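# Master material for one-sided fog sheets, translucent and unlit, with switchable features.
# Two panning cloud textures, one bump-offset, blend into an overlay that can modulate the falloff-texture opacity and tint the squared color.
# Opacity also fades near the camera, with distance, and by view angle against world Z.
# A final switch selects DepthFade or pixel-depth density, replacing UE3 expressions; the Density parameter is left unconnected.
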Material /Engine/EngineVolumetrics/Fogsheet/Materials/M_EV_FogSheet_1sided_Master_01
graph {
 properties {bUsedWithStaticLighting=true, blendMode=BLEND_Translucent, pixelDepthOffsetMode=PDOM_Legacy, shadingModel=MSM_Unlit, usedShadingModels=Unlit}
 nodes [
  ScalarParameter() {defaultValue=0.5, parameterName=Density}
  n1 = VectorParameter() {desc="Color (A is luminosity)", parameterName=Color, defaultValue=<rgba 1 1 1 1>}
  n2 = Multiply(A=@n1.RGB, B=@n1.RGB)
  n3 = TextureCoordinate() {uTiling=3.2, vTiling=3.2}
  n4 = Constant() {r=-20}
  n5 = ScalarParameter() {defaultValue=1, parameterName=TextureScale}
  n6 = StaticSwitchParameter(True=AppendVector(A=@n5, B=ScalarParameter() {defaultValue=1, parameterName=TextureScaleY}), False=@n5) {parameterName=NonUniformScale}
  n7 = Multiply(A=Time(), B=ScalarParameter() {defaultValue=1, parameterName=PanningSpeed})
  n8 = Panner(Coordinate=Multiply(A=BumpOffset(Coordinate=@n3, Height=@n4), B=@n6), Time=@n7) {speedX=0.03, speedY=-0.015}
  n9 = Panner(Coordinate=Multiply(A=TextureCoordinate() {uTiling=3, vTiling=3}, B=@n6), Time=@n7) {speedX=-0.01, speedY=-0.02}
  n10 = LinearInterpolate(
   A=Constant() {r=1}
   B=LinearInterpolate(
    A=TextureSampleParameter2D(UVs=@n8) {parameterName=MultiplyTexture1, texture=&/T_EV_FogSheet_CloudTex_02}.RGB
    B=TextureSampleParameter2D(UVs=@n9) {parameterName=MultiplyTexture2, texture=&T_EV_FogSheet_CloudTex_02}.RGB
    Alpha=Constant() {r=0.5}
   )
   Alpha=ScalarParameter() {defaultValue=1, parameterName=TextureContrast}
  )
  n11 = StaticSwitchParameter(
   True=@n10
   False=TextureSampleParameter2D() {parameterName=SimpleEmissiveTex, texture=/Engine/EngineVolumetrics/Fogsheet/T_EV_Fogsheet_FlareColors_01}.RGB
  ) {parameterName=UseOpacityTexasEmissive}
  n12 = ScalarParameter() {defaultValue=1, parameterName=EmissiveTexAdd}
  StaticSwitchParameter(True=Multiply(A=@n2, B=Add(A=@n11, B=@n12)), False=@n2) {parameterName=UseEmissiveTexture}
  n13 = StaticSwitchParameter(
   True=TextureSampleParameter2D() {parameterName=FalloffTexture, texture=&/T_EV_FogSheet_Falloff_01}.G
   False=Constant() {r=1}
  ) {defaultValue=true, parameterName=UseAlphaTexture}
  n14 = StaticSwitchParameter(True=Multiply(A=@n10, B=@n13), False=@n13) {parameterName=UseTextureOverlay}
  n15 = Multiply(A=PixelDepth().R, B=Divide(A=Constant() {r=0.0005}, B=ScalarParameter() {defaultValue=1, parameterName=Distance}))
  n16 = Constant() {r=1}
  n17 = Divide(A=Constant() {r=0.0025}, B=ScalarParameter() {defaultValue=1, parameterName=FarFadeDistance})
  n18 = PixelDepth()
  n19 = OneMinus(None=Clamp(None=Multiply(A=@n17, B=@n18.R), Max=Constant() {r=1}))
  n20 = Constant() {r=1}
  n21 = CameraVectorWS()
  n22 = Constant3Vector() {constant=<rgba 0 0 1 0>}
  n23 = Power(Base=Abs(None=DotProduct(A=@n21, B=@n22)), Exp=ScalarParameter() {defaultValue=3, parameterName=FalloffExponent})
  n24 = Multiply(
   A=StaticSwitchParameter(True=Clamp(None=@n15, Max=@n16), False=@n16) {defaultValue=true, parameterName=Fade when Near Camera}
   B=Multiply(
    A=StaticSwitchParameter(True=@n19, False=@n20) {parameterName=Fade with Distance}
    B=StaticSwitchParameter(True=@n23, False=@n20) {defaultValue=true, parameterName=Use Edge Falloff}
   )
  )
  n25 = Clamp(None=Multiply(A=Multiply(A=@n14, B=@n24), B=ScalarParameter() {defaultValue=1, desc=Opacity, parameterName=Opacity}), Max=Constant() {r=1})
  n26 = Clamp(None=PixelDepth() {desc=UE3 used DestDepth expression}.R) {maxDefault=65535}
  n27 = ScalarParameter() {parameterName=StartDistance}
  n28 = Clamp(None=Multiply(A=Add(A=@n26, B=@n27), B=ScalarParameter() {defaultValue=0.0001, parameterName=DepthBasedDensity}))
  StaticSwitchParameter(
   True=StaticSwitchParameter(
    True=Multiply(A=@n25, B=@n28)
    False=DepthFade(Opacity=@n25) {desc="UE3 used DepthBiasedAlpha expression, and had the Density param hookedup", fadeDistanceDefault=5000}
   ) {parameterName=UseSceneDepthBlend}
   False=@n25
  ) {defaultValue=true, parameterName=UseDepthBiasedAlpha}
 ]
}
