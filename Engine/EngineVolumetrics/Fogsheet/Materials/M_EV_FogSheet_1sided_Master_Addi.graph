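# Additive, unlit master material for one-sided volumetric fog sheets, with static switches for its optional features.
# Color is squared and can be modulated by an emissive texture.
# Opacity combines a falloff texture or two blended panning cloud textures with near-camera, far-distance and view-angle fades.
# An optional final depth stage uses DepthFade or, with UseSceneDepthBlend on, pixel-depth-based density; the Density parameter is unconnected.
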Material /Engine/EngineVolumetrics/Fogsheet/Materials/M_EV_FogSheet_1sided_Master_Addi
graph {
 properties {blendMode=BLEND_Additive, pixelDepthOffsetMode=PDOM_Legacy, shadingModel=MSM_Unlit, usedShadingModels=Unlit}
 nodes [
  ScalarParameter() {defaultValue=0.5, parameterName=Density}
  n1 = VectorParameter() {desc="Color (A is luminosity)", parameterName=Color, defaultValue=<rgba 1 1 1 1>}
  n2 = Multiply(A=@n1.RGB, B=@n1.RGB)
  n3 = TextureSampleParameter2D() {parameterName=SimpleEmissiveTex, texture=/Engine/EngineVolumetrics/Fogsheet/T_EV_Fogsheet_FlareColors_01}
  n4 = ScalarParameter() {defaultValue=1, parameterName=EmissiveTexAdd}
  StaticSwitchParameter(True=Multiply(A=@n2, B=Add(A=@n3.RGB, B=@n4)), False=@n2) {parameterName=UseEmissiveTexture}
  n5 = Constant() {r=1}
  n6 = ScalarParameter() {defaultValue=1, parameterName=TextureScale}
  n7 = StaticSwitchParameter(True=AppendVector(A=@n6, B=ScalarParameter() {defaultValue=1, parameterName=TextureScaleY}), False=@n6) {parameterName=NonUniformScale}
  n8 = Multiply(A=TextureCoordinate() {uTiling=4, vTiling=4}, B=@n7)
  n9 = Multiply(A=Time(), B=ScalarParameter() {defaultValue=1, parameterName=PanningSpeed})
  n10 = Multiply(A=TextureCoordinate() {uTiling=3, vTiling=3}, B=@n7)
  n11 = LinearInterpolate(
   A=TextureSampleParameter2D(UVs=Panner(Coordinate=@n8, Time=@n9) {speedX=0.03, speedY=-0.015}) {parameterName=MultiplyTexture1, texture=&/T_EV_FogSheet_CloudTex_02}.RGB
   B=TextureSampleParameter2D(UVs=Panner(Coordinate=@n10, Time=@n9) {speedX=-0.01, speedY=-0.02}) {parameterName=MultiplyTexture2, texture=&T_EV_FogSheet_CloudTex_02}.RGB
   Alpha=Constant() {r=0.5}
  )
  n12 = ScalarParameter() {defaultValue=1, parameterName=TextureContrast}
  n13 = TextureSampleParameter2D() {parameterName=FalloffTexture, texture=&/T_EV_FogSheet_Falloff_01}
  n14 = Constant() {r=1}
  n15 = StaticSwitchParameter(
   True=Multiply(
    A=LinearInterpolate(A=@n5, B=@n11, Alpha=@n12)
    B=StaticSwitchParameter(True=@n13.G, False=@n14) {defaultValue=true, parameterName=UseAlphaTexture}
   )
   False=@n13.G
  ) {parameterName=UseTextureOverlay}
  n16 = Multiply(A=PixelDepth().R, B=Divide(A=Constant() {r=0.0005}, B=ScalarParameter() {defaultValue=1, parameterName=Distance}))
  n17 = Constant() {r=1}
  n18 = Divide(A=Constant() {r=0.0025}, B=ScalarParameter() {defaultValue=1, parameterName=FarFadeDistance})
  n19 = PixelDepth()
  n20 = OneMinus(None=Clamp(None=Multiply(A=@n18, B=@n19.R), Max=Constant() {r=1}))
  n21 = Constant() {r=1}
  n22 = CameraVectorWS()
  n23 = Constant3Vector() {constant=<rgba 0 0 1 0>}
  n24 = Power(Base=Abs(None=DotProduct(A=@n22, B=@n23)), Exp=ScalarParameter() {defaultValue=3, parameterName=FalloffExponent})
  n25 = Multiply(
   A=StaticSwitchParameter(True=Clamp(None=@n16, Max=@n17), False=@n17) {defaultValue=true, parameterName=Fade when Near Camera}
   B=Multiply(
    A=StaticSwitchParameter(True=@n20, False=@n21) {parameterName=Fade with Distance}
    B=StaticSwitchParameter(True=@n24, False=@n21) {defaultValue=true, parameterName=Use Edge Falloff}
   )
  )
  n26 = Clamp(None=Multiply(A=Multiply(A=@n15, B=@n25), B=ScalarParameter() {defaultValue=1, desc=Opacity, parameterName=Opacity}), Max=Constant() {r=1})
  n27 = Clamp(None=PixelDepth().R) {maxDefault=65535}
  n28 = ScalarParameter() {parameterName=StartDistance}
  n29 = Clamp(None=Multiply(A=Add(A=@n27, B=@n28), B=ScalarParameter() {defaultValue=0.0001, parameterName=DepthBasedDensity}))
  StaticSwitchParameter(
   True=StaticSwitchParameter(
    True=Multiply(A=@n26, B=@n29)
    False=DepthFade(Opacity=@n26) {desc="UE3 used DepthBiasedAlpha expression, and had the Density param hookedup", fadeDistanceDefault=5000}
   ) {parameterName=UseSceneDepthBlend}
   False=@n26
  ) {defaultValue=true, parameterName=UseDepthBiasedAlpha}
 ]
}
