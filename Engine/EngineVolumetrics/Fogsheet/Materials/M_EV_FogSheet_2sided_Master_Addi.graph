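# Two-sided, unlit additive master material for fog sheets.
# Two panning cloud textures optionally overlay a falloff texture, scaled by switchable near, far and view-to-Z-axis edge fades.
# The resulting opacity passes through switchable DepthFade or pixel-depth-based density, while squared Color RGB, optionally modulated by an emissive texture, supplies the tint.
# UE3 porting notes mark the depth nodes, and the Density parameter is left unconnected.
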Material /Engine/EngineVolumetrics/Fogsheet/Materials/M_EV_FogSheet_2sided_Master_Addi
graph {
 properties {blendMode=BLEND_Additive, pixelDepthOffsetMode=PDOM_Legacy, shadingModel=MSM_Unlit, twoSided=true, usedShadingModels=Unlit}
 nodes [
  ScalarParameter() {defaultValue=0.5, parameterName=Density}
  n1 = Constant() {r=1}
  n2 = ScalarParameter() {defaultValue=1, parameterName=TextureScale}
  n3 = StaticSwitchParameter(True=AppendVector(A=@n2, B=ScalarParameter() {defaultValue=1, parameterName=TextureScaleY}), False=@n2) {parameterName=NonUniformScale}
  n4 = Multiply(A=TextureCoordinate() {uTiling=4, vTiling=4}, B=@n3)
  n5 = Multiply(A=Time(), B=ScalarParameter() {defaultValue=1, parameterName=PanningSpeed})
  n6 = Multiply(A=TextureCoordinate() {uTiling=3, vTiling=3}, B=@n3)
  n7 = LinearInterpolate(
   A=TextureSampleParameter2D(UVs=Panner(Coordinate=@n4, Time=@n5) {speedX=0.03, speedY=-0.015}) {parameterName=MultiplyTexture1, texture=&/T_EV_FogSheet_CloudTex_02}.RGB
   B=TextureSampleParameter2D(UVs=Panner(Coordinate=@n6, Time=@n5) {speedX=-0.01, speedY=-0.02}) {parameterName=MultiplyTexture2, texture=&T_EV_FogSheet_CloudTex_02}.RGB
   Alpha=Constant() {r=0.5}
  )
  n8 = ScalarParameter() {defaultValue=1, parameterName=TextureContrast}
  n9 = TextureSampleParameter2D() {parameterName=FalloffTexture, texture=&/T_EV_FogSheet_Falloff_01}
  n10 = Constant() {r=1}
  n11 = StaticSwitchParameter(
   True=Multiply(
    A=LinearInterpolate(A=@n1, B=@n7, Alpha=@n8)
    B=StaticSwitchParameter(True=@n9.G, False=@n10) {defaultValue=true, parameterName=UseAlphaTexture}
   )
   False=@n9.G
  ) {parameterName=UseTextureOverlay}
  n12 = Multiply(A=PixelDepth().R, B=Divide(A=Constant() {r=0.0005}, B=ScalarParameter() {defaultValue=1, parameterName=Distance}))
  n13 = Constant() {r=1}
  n14 = Divide(A=Constant() {r=0.0025}, B=ScalarParameter() {defaultValue=1, parameterName=FarFadeDistance})
  n15 = PixelDepth()
  n16 = OneMinus(None=Clamp(None=Multiply(A=@n14, B=@n15.R), Max=Constant() {r=1}))
  n17 = Constant() {r=1}
  n18 = CameraVectorWS()
  n19 = Constant3Vector() {constant=<rgba 0 0 1 0>}
  n20 = Power(Base=Abs(None=DotProduct(A=@n18, B=@n19)), Exp=ScalarParameter() {defaultValue=3, parameterName=FalloffExponent})
  n21 = Multiply(
   A=StaticSwitchParameter(True=Clamp(None=@n12, Max=@n13), False=@n13) {defaultValue=true, parameterName=Fade when Near Camera}
   B=Multiply(
    A=StaticSwitchParameter(True=@n16, False=@n17) {parameterName=Fade with Distance}
    B=StaticSwitchParameter(True=@n20, False=@n17) {defaultValue=true, parameterName=Use Edge Falloff}
   )
  )
  n22 = Clamp(None=Multiply(A=Multiply(A=@n11, B=@n21), B=ScalarParameter() {defaultValue=1, desc=Opacity, parameterName=Opacity}), Max=Constant() {r=1})
  n23 = Clamp(None=PixelDepth() {desc=UE3 used DestDepth expression}.R) {maxDefault=65535}
  n24 = ScalarParameter() {parameterName=StartDistance}
  n25 = Clamp(None=Multiply(A=Add(A=@n23, B=@n24), B=ScalarParameter() {defaultValue=0.0001, parameterName=DepthBasedDensity}))
  StaticSwitchParameter(
   True=StaticSwitchParameter(
    True=Multiply(A=@n22, B=@n25)
    False=DepthFade(Opacity=@n22) {desc="UE3 used DepthBiasedAlpha expression, and had the Density param hookedup", fadeDistanceDefault=5000}
   ) {parameterName=UseSceneDepthBlend}
   False=@n22
  ) {defaultValue=true, parameterName=UseDepthBiasedAlpha}
  n26 = VectorParameter() {desc="Color (A is luminosity)", parameterName=Color, defaultValue=<rgba 1 1 1 1>}
  n27 = Multiply(A=@n26.RGB, B=@n26.RGB)
  n28 = TextureSampleParameter2D() {parameterName=SimpleEmissiveTex, texture=/Engine/EngineVolumetrics/Fogsheet/T_EV_Fogsheet_FlareColors_01}
  n29 = ScalarParameter() {defaultValue=1, parameterName=EmissiveTexAdd}
  StaticSwitchParameter(True=Multiply(A=@n27, B=Add(A=@n28.RGB, B=@n29)), False=@n27) {parameterName=UseEmissiveTexture}
 ]
}
